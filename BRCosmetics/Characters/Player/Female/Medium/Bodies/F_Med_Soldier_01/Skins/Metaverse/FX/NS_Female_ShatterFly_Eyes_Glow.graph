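# Niagara cosmetic effect that puts a cyan glow on both eyes of the female ShatterFly Metaverse skin.
# A single emitter bursts two persistent sprites, each assigned a spawn group that selects a mirrored offset from the head bone.
# Skeletal mesh sampling updates their position as the head moves, with a camera offset.
# Loop-age curves subtly pulse alpha and size, and a time-of-day scalar dims alpha at night.
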NiagaraSystem /BRCosmetics/Characters/Player/Female/Medium/Bodies/F_Med_Soldier_01/Skins/Metaverse/FX/NS_Female_ShatterFly_Eyes_Glow
dependencies {
 usedDataInterfaces [&NiagaraDataInterfaceCurve, &NiagaraDataInterfaceSkeletalMesh, &NiagaraDataInterfaceArrayFloat3]
 usedDynamicInputs []
 usedModules [&SystemState]
}
emitters {eyes_light {
  emitterUpdateScript [
   &EmitterState(
    Inactive Response=::NewEnumerator1
    Life Cycle Mode=::NewEnumerator1
    Loop Behavior=::NewEnumerator1
    Loop Duration Mode=::NewEnumerator1
    Reset Age On Awaken=true
    Scalability Mode=::NewEnumerator1
    UseLoopDelay=false
   )
   &SpawnBurst_Instantaneous(Spawn Count=2, Use Spawn Probability=false)
  ]
  particleSpawnScript [
   &FN_SetSpawnGroup(Number of Spawn Groups=2)
   &InitializeParticle(
    Color=<rgba 0 1.3085 2 1>
    Sprite Size=&Multiply_Vector2D(A=<13 10>, B=&Multiply_Vector2D_ByFloat(Float=0.7, Vector2D=<1 1>))
    Sprite Size Mode=::NewEnumerator4
    UsePositionOffset=false
    Write Material Random=true
   ) {module=InitializeParticle001}
  ]
  particleUpdateScript [
   &ParticleState(Kill Particles When Lifetime Has Elapsed=false, Let Infinitely Lived Particles Die When Emitter Deactivates=true)
   &ScaleColor(
    Scale Alpha=&Multiply_Float(
     A=&Multiply_Float(A=1, B=&FNTimeOfDayScalar(Night Scale=0.8, Morning Scale=1, Evening Scale=1))
     B=&FloatFromCurve(
      FloatCurve=DataInterface() {propertyValues {Curve {keys [<at 0 1>,<at 0.0539 1>,<at 0.1557 0.8>,<at 0.2614 1>,<at 0.3742 0.8>,<at 0.456 1>,<at 0.5872 0.8>,<at 0.7535 1>,<at 0.8976 0.8>,<at 1 1>]}}}
      CurveIndex=$System.NormalizedLoopAge
     )
    )
    ScaleA=true
    ScaleRGB=true
   )
   &SolveForcesAndVelocity(Clamp Velocity=false, Limit Acceleration=false)
   &CameraOffset(Camera Offset Amount=1.1)
   &SkeletalMeshLocation(
    Apply Sampled Orientation to Mesh Particles=true
    Bone / Socket Index=$Particles.SpawnGroup
    Bone Sampling Mode=::NewEnumerator4
    Inherit Velocity=false
    Interpolate Mesh Sampling=false
    MaskSpawnGroup=false
    Sampled Position Offset=&SelectVectorFromArray(
     Direct Array Index=$Particles.SpawnGroup
     Vector Selection Array=DataInterface() {propertyValues {FloatData [<3.5 6.7 2.8>, <3.5 6.7 -2.8>]}}
    )
    Sampled Position Offset Coordinate Space:ENiagaraMeshTransforms=::NewEnumerator3
    Skeletal Mesh=DataInterface() {propertyValues {
      DefaultMesh=null
      ExcludeBoneName=root
      OptionalComponentClass=null
      PreviewMesh=/Game/Characters/Player/Female/Medium/Heads/F_MED_ASN_Sarah_Head_01/Meshes/F_MED_ASN_Sarah_Head_01
      SkinningMode=SkinOnTheFly
      SoftSourceActor=null
      SourceMode=Default
      UvSetIndex=0
      WholeMeshLOD=-1
      bExcludeBone=false
      bReadDeformedGeometry=true
      bRequireCurrentFrameData=true
      ComponentTags []
      FilteredBones [head,head]
      FilteredSockets []
      MeshUserParameter {parameter {name=None, typeDefHandle {registeredTypeIndex=1}, varData []}}
      SamplingRegions []
     }}
   ) {module=SkeletalMeshLocation001}
   &DynamicMaterialParameters(
    Index 0 Param 1=1
    Index 0 Param 2=1
    Index 0 Param 3=1
    Index 0 Param 4=1
    Index 1 Param 1=1
    Index 1 Param 2=1
    Index 1 Param 3=1
    Index 1 Param 4=1
    Module.Param0WriteEnabled=true
    Module.Param0WriteEnabled1=true
    Module.Param1WriteEnabled=true
    Module.Param1WriteEnabled1=true
    Module.Param2WriteEnabled=true
    Module.Param2WriteEnabled1=true
    Module.Param3WriteEnabled=true
    Module.Param3WriteEnabled1=true
    Write Parameter Index 1=true
   )
   &ScaleSpriteSize(Scale Factor=&Vector2DFromFloat(Value=&FloatFromCurve(
      FloatCurve=DataInterface() {propertyValues {Curve {keys [<at 0 1>,<at 0.0539 1>,<at 0.1557 0.985>,<at 0.2614 1>,<at 0.3742 0.985>,<at 0.456 1>,<at 0.5872 0.985>,<at 0.7535 1>,<at 0.8976 0.985>,<at 1 1>]}}}
      CurveIndex=$System.NormalizedLoopAge
     )))
  ]
  rendererClasses [&NiagaraSpriteRendererProperties]
  renderers [{rendererClass=&NiagaraSpriteRendererProperties}]
 }}
summary {
 systemName=NS_Female_ShatterFly_Eyes_Glow
 emitters [{emitterName=eyes_light, rendererClasses [&NiagaraSpriteRendererProperties]}]
 userVariables []
}
